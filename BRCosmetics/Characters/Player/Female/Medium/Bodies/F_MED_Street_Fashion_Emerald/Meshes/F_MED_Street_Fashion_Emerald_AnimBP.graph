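# Post-pose AnimBP for the female Street Fashion Emerald body, adding secondary motion over the pose copied from the attached parent mesh.
# A weighted constraint places the grenade bone, thigh swing angles drive additive subthigh scaling, and spring-based AnimDynamics sways four jacket bones.
# The EventGraph also computes backpack, LOD/rigid-body and head-rotation values, which the AnimGraph leaves unconnected.
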AnimBlueprint /BRCosmetics/Characters/Player/Female/Medium/Bodies/F_MED_Street_Fashion_Emerald/Meshes/F_MED_Street_Fashion_Emerald_AnimBP
target_skeleton=&/F_MED_Street_Fashion_Emerald_Skeleton
graphs [
 AnimGraph {nodes [
   n1 = "Constraint - dyn_grenade (PPPP)"!(
    ComponentPose=Local To Component(LocalPose=Copy Pose From Mesh() {bUseAttachedParent=true})
    Alpha=1
    ConstraintWeights_0=0.4754
    ConstraintWeights_1:float=0.0527
    ConstraintWeights_2:float=0.0007
    ConstraintWeights_3:float=0.4712
   ) -> (Pose) {
    alphaBoolBlend {blendOption=Linear}
    boneToModify {boneName=dyn_grenade}
    constraintSetup [
     {targetBone {boneName=thigh_r}}
     {targetBone {boneName=pelvis}}
     {targetBone {boneName=thigh_twist_01_l}}
     {targetBone {boneName=thigh_twist_01_r}}
    ]
    constraintWeights [1,1,1,1]
   }
   n2 = "Transform (Modify) Bone - Bone: dyn_subthigh_l"(ComponentPose=@n1, Alpha=Get legSwingDelta_l() -> (legSwingDelta_l:double), Translation="0,0,0", Rotation="0,0,0", Scale="4,1,4") {
    scaleMode=BMM_Additive
    scaleSpace=BCS_BoneSpace
    translationMode=BMM_Additive
    translationSpace=BCS_BoneSpace
    alphaBoolBlend {blendOption=Linear}
    boneToModify {boneName=dyn_subthigh_l}
    scale {x=4, z=4}
   }
   n3 = "Transform (Modify) Bone - Bone: dyn_subthigh_r"(ComponentPose=@n2, Alpha=Get legSwingDelta_r() -> (legSwingDelta_r:double), Translation="0,0,0", Rotation="0,0,0", Scale="2.25,1,2.25") {
    scaleMode=BMM_Additive
    scaleSpace=BCS_BoneSpace
    translationMode=BMM_Additive
    translationSpace=BCS_BoneSpace
    alphaBoolBlend {blendOption=Linear}
    boneToModify {boneName=dyn_subthigh_r}
    scale {x=2.25, z=2.25}
   }
   n4 = Anim Dynamics - Bone: dyn_jacket_r(ComponentPose=@n3, Alpha=1, ExternalForce="0,0,0") {
    angularDampingOverride=1
    angularSpringConstant=1500
    bAngularSpring=true
    bGravityOverrideInSimSpace=true
    bOverrideAngularDamping=true
    gravityScale=0.35
    simulationSpace=RootRelative
    alphaBoolBlend {blendOption=Linear}
    alphaScaleBiasClamp {scale=0.4}
    boundBone {boneName=dyn_jacket_r}
    componentAppliedLinearAccClamp {z=5000}
    componentLinearAccScale {z=0.45}
    componentLinearVelScale {z=0.45}
    physicsBodyDefinitions [{
      sphereCollisionRadius=0
      boundBone {boneName=dyn_jacket_r}
      boxExtents=<10 20 20>
      constraintSetup {angularLimitsMax=<0 30 0>, angularLimitsMin=<0 -30 -60>, angularTarget=<0 0 0>}
      localJointOffset=<10 0 0>
     }]
   }
   n5 = Anim Dynamics - Bone: dyn_jacket_back(ComponentPose=@n4, Alpha=1, ExternalForce="0,0,0") {
    angularDampingOverride=1
    angularSpringConstant=1500
    bAngularSpring=true
    bGravityOverrideInSimSpace=true
    bOverrideAngularDamping=true
    gravityScale=0.35
    simulationSpace=RootRelative
    alphaBoolBlend {blendOption=Linear}
    alphaScaleBiasClamp {scale=0.3}
    boundBone {boneName=dyn_jacket_back}
    componentAppliedLinearAccClamp {z=5000}
    componentLinearAccScale {z=0.45}
    componentLinearVelScale {z=0.45}
    physicsBodyDefinitions [{
      sphereCollisionRadius=0
      boundBone {boneName=dyn_jacket_back}
      boxExtents=<10 20 20>
      constraintSetup {angularLimitsMax=<0 30 0>, angularLimitsMin=<0 -30 -60>, angularTarget=<0 0 0>}
      localJointOffset=<10 0 0>
     }]
   }
   n6 = Anim Dynamics - Bone: dyn_jacket_front(ComponentPose=@n5, Alpha=1, ExternalForce="0,0,0") {
    angularDampingOverride=1
    angularSpringConstant=1500
    bAngularSpring=true
    bGravityOverrideInSimSpace=true
    bOverrideAngularDamping=true
    gravityScale=0.35
    simulationSpace=RootRelative
    alphaBoolBlend {blendOption=Linear}
    alphaScaleBiasClamp {scale=0.3}
    boundBone {boneName=dyn_jacket_front}
    componentAppliedLinearAccClamp {z=5000}
    componentLinearAccScale {z=0.45}
    componentLinearVelScale {z=0.45}
    physicsBodyDefinitions [{
      sphereCollisionRadius=0
      boundBone {boneName=dyn_jacket_front}
      boxExtents=<10 20 20>
      constraintSetup {angularLimitsMax=<0 30 0>, angularLimitsMin=<0 -30 -60>, angularTarget=<0 0 0>}
      localJointOffset=<10 0 0>
     }]
   }
   n7 = Anim Dynamics - Bone: dyn_jacket_l(ComponentPose=@n6, Alpha=1, ExternalForce="0,0,0") {
    angularDampingOverride=1
    angularSpringConstant=1500
    bAngularSpring=true
    bGravityOverrideInSimSpace=true
    bOverrideAngularDamping=true
    gravityScale=0.35
    simulationSpace=RootRelative
    alphaBoolBlend {blendOption=Linear}
    alphaScaleBiasClamp {scale=0.4}
    boundBone {boneName=dyn_jacket_l}
    componentAppliedLinearAccClamp {z=5000}
    componentLinearAccScale {z=0.45}
    componentLinearVelScale {z=0.45}
    physicsBodyDefinitions [{
      sphereCollisionRadius=0
      boundBone {boneName=dyn_jacket_l}
      boxExtents=<10 20 20>
      constraintSetup {angularLimitsMax=<0 30 0>, angularLimitsMin=<0 -30 -60>, angularTarget=<0 0 0>}
      localJointOffset=<10 0 0>
     }]
   }
   Output Pose(Result=Component To Local(ComponentPose=@n7)) {layerGroup=None}
   Get isBackStatic() -> (isBackStatic:bool)
  ]}
 EventGraph {nodes [
   Event BlueprintUpdateAnimation()
   > n1 = Sequence!(execute) -> (then_0, then_1, then_2:Exec, then_3:Exec)
   n2 = Get player_pawn() -> (player_pawn:Fort Player Pawn Athena Object Reference)
   > Is Valid(InputObject=@n2)
   > Do Once()
   > Set isBackpackEquipped(isBackpackEquipped:bool="Not Equal (Object)"!(A:Object Reference=GetSkeletalMeshForPartType(self=@n2, PartType=Backpack), B:Object Reference) -> (ReturnValue:bool)) -> (Output_Get:bool)
   n3 = GetOwnerSkeletalMeshComponent()
   n4 = BreakTransform(InTransform=GetDeltaTransformFromRefPose(self=@n3, BoneName=head, BaseName=spine_05))
   Set head_rot_back(execute=@n1.then_2, head_rot_back:double=MapRangeClamped(Value=BreakRotator(InRot=@n4.Rotation).Yaw, InRangeA=0, InRangeB=-48)) -> (Output_Get:double)
   n5 = "integer > integer"(A=Get CurrentLOD(), B=1) {is=title K2Node_CallFunction}
   n6 = OR Boolean(A=NOT Boolean(A=Get bRigidBodyEnabled()), B=@n5)
   Set isRigidBodyDisabled(execute=@n1.then_1, isRigidBodyDisabled:bool=@n6) -> (Output_Get:bool)
   > Set isBackStatic(isBackStatic:bool=OR Boolean(A=Get isBackpackEquipped() -> (isBackpackEquipped:bool), B=@n6)) -> (Output_Get:bool)
   > Set Rigids_lod_offset(Rigids_lod_offset:Vector=SelectVector(A="0,-5, 0", B="0, 0, 0", bPickA=@n5)) -> (Output_Get:Vector)
   > Set Rigids_lod_rot_offset(Rigids_lod_rot_offset:Rotator=SelectRotator!(A:Rotator="0,20, 0", B:Rotator="0, 0, 0", bPickA:bool=@n5) -> (ReturnValue:Rotator)) -> (Output_Get:Rotator)
   Set legSwingDelta_r(
    execute=@n1.then_3
    legSwingDelta_r:double=MapRangeClamped(Value=GetTwistAndSwingAngleOfDeltaRotationFromRefPose(self=@n3, BoneName=thigh_r).OutSwingAngle, InRangeA=150, InRangeB=75)
   ) -> (Output_Get:double)
   > Set legSwingDelta_l(legSwingDelta_l:double=MapRangeClamped(Value=GetTwistAndSwingAngleOfDeltaRotationFromRefPose(self=@n3, BoneName=thigh_l).OutSwingAngle, InRangeA=43, InRangeB=110)) -> (Output_Get:double)
   Event BlueprintInitializeAnimation()
   n7 = TryGetPawnOwner()
   > Is Valid(InputObject=@n7)
   > n8 = Cast To FortPlayerPawnAthena(Object=@n7)
   > Set player_pawn(player_pawn:Fort Player Pawn Athena Object Reference=@n8) -> (Output_Get:Fort Player Pawn Athena Object Reference)
   TryGetPawnOwner()
  ]}
]
variables [
 head_rot_back {type=number}
 isBackpackEquipped {type=boolean}
 isBackStatic {type=boolean}
 isRigidBodyDisabled {type=boolean}
 legSwingDelta_l {type=number}
 legSwingDelta_r {type=number}
 player_pawn {type=FortPlayerPawnAthena}
 Rigids_lod_offset {type=Vector}
 Rigids_lod_rot_offset {type=Rotator}
]
