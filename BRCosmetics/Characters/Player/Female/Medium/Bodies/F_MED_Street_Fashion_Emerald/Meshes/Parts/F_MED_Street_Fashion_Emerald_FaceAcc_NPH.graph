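# Physics asset for the face-accessory part of the female medium Street Fashion Emerald body, providing secondary motion for dynamic hair bones.
# Five three-bone chains of box bodies hang from the head with limited swing and tight twist.
# Angularly free constraints cross-link neighbouring chains at middle segments, plus back chains at tips, tying the hair together.
# Capsules on head, neck, upper spine and clavicles supply body collision shapes.
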PhysicsAsset /BRCosmetics/Characters/Player/Female/Medium/Bodies/F_MED_Street_Fashion_Emerald/Meshes/Parts/F_MED_Street_Fashion_Emerald_FaceAcc_NPH
bodies {
 clavicle_l [{shapeType=Capsule, length=11.7525, radius=6.9568, center=<5.5042 -0.5608 -1.9968>, rotation=<rot 74.9964 7.1654 0>}]
 clavicle_r [{shapeType=Capsule, length=11.7525, radius=6.9568, center=<-5.2943 0.5714 2.0109>, rotation=<rot 74.9972 7.1655 0>}]
 dyn_hair_bk_01 [{shapeType=Box, extentX=8.5455, extentY=7.9833, extentZ=1.4662, center=<4.6165 0 0>}]
 dyn_hair_bk_02 [{shapeType=Box, extentX=5.5455, extentY=14.4833, extentZ=1.4662, center=<2.77 0 0.0001>, rotation=<rot 0 0 5>}]
 dyn_hair_bk_03 [{shapeType=Box, extentX=5.5455, extentY=16.9833, extentZ=1.4662, center=<2.4661 0 -0.5614>, rotation=<rot -14.9999 0 5.0002>}]
 dyn_hair_bk_l_01 [{shapeType=Box, extentX=5.5455, extentY=7.9833, extentZ=1.4662, center=<2.6166 0 0.0001>}]
 dyn_hair_bk_l_02 [{shapeType=Box, extentX=5.5455, extentY=7.9833, extentZ=1.4662, center=<2.6166 0 0.0001>}]
 dyn_hair_bk_l_03 [{shapeType=Box, extentX=3.0455, extentY=4.4833, extentZ=1.4662, center=<1.7724 0.0001 1.5146>, rotation=<rot 49.9999 0 0>}]
 dyn_hair_bk_r_01 [{shapeType=Box, extentX=8.0455, extentY=7.9833, extentZ=1.4662, center=<4.5991 0 -0.0001>}]
 dyn_hair_bk_r_02 [{shapeType=Box, extentX=7.5455, extentY=15.4833, extentZ=1.4662, center=<4.1569 -1.3669 -1.0456>, rotation=<rot 20.0003 -20.0008 -14.9991>}]
 dyn_hair_bk_r_03 [{shapeType=Box, extentX=6.5455, extentY=11.4833, extentZ=1.4662, center=<3.4339 -0.002 -1.0299>, rotation=<rot -25.0003 0.0002 -5.0002>}]
 dyn_hair_l_01 [{shapeType=Box, extentX=5.5455, extentY=7.9833, extentZ=1.4662, center=<2.6166 0 0.0001>}]
 dyn_hair_l_02 [{shapeType=Box, extentX=5.5455, extentY=7.9833, extentZ=1.4662, center=<2.6166 0 0.0001>}]
 dyn_hair_l_03 [{shapeType=Box, extentX=3.0455, extentY=5.9833, extentZ=1.4662, center=<1.6889 0 0.0001>}]
 dyn_hair_r_01 [{shapeType=Box, extentX=5.5455, extentY=7.9833, extentZ=1.4662, center=<2.6166 0 0.0001>}]
 dyn_hair_r_02 [{shapeType=Box, extentX=5.5455, extentY=10.9833, extentZ=1.4662, center=<2.6166 0 0.0001>}]
 dyn_hair_r_03 [{shapeType=Box, extentX=3.5455, extentY=15.4833, extentZ=1.4662, center=<2.6166 1.6463 -0.1154>, rotation=<rot 0.0024 0.0009 -14.9997>}]
 head [{shapeType=Capsule, length=4.25, radius=5.75, center=<1.0525 1.6777 -0.0001>, rotation=<rot -90 0.0098 -0.0098>}]
 neck_01 [{shapeType=Capsule, length=3.5, radius=6.5, center=<2.5231 1.6777 0>, rotation=<rot -90 -0.005 0.005>}]
 spine_05 [{shapeType=Capsule, length=4.6562, radius=7.25, center=<6.8721 2.3603 0>}]
}
constraints [
 {
  bone1Name=clavicle_l
  bone2Name=spine_05
  swing1LimitDegrees=45
  swing1Motion=Free
  swing2LimitDegrees=45
  swing2Motion=Free
  twistLimitDegrees=45
  twistMotion=Free
 }
 {
  bone1Name=clavicle_r
  bone2Name=spine_05
  swing1LimitDegrees=45
  swing1Motion=Free
  swing2LimitDegrees=45
  swing2Motion=Free
  twistLimitDegrees=45
  twistMotion=Free
 }
 {bone1Name=dyn_hair_l_01, bone2Name=head, swing1LimitDegrees=45, swing2LimitDegrees=45, twistLimitDegrees=5}
 {bone1Name=dyn_hair_l_02, bone2Name=dyn_hair_l_01, swing1LimitDegrees=45, swing2LimitDegrees=45, twistLimitDegrees=5}
 {bone1Name=dyn_hair_l_03, bone2Name=dyn_hair_l_02, swing1LimitDegrees=45, swing2LimitDegrees=45, twistLimitDegrees=5}
 {bone1Name=dyn_hair_bk_r_01, bone2Name=head, swing1LimitDegrees=45, swing2LimitDegrees=45, twistLimitDegrees=5}
 {bone1Name=dyn_hair_bk_r_02, bone2Name=dyn_hair_bk_r_01, swing1LimitDegrees=45, swing2LimitDegrees=45, twistLimitDegrees=5}
 {bone1Name=dyn_hair_bk_r_03, bone2Name=dyn_hair_bk_r_02, swing1LimitDegrees=45, swing2LimitDegrees=45, twistLimitDegrees=5}
 {bone1Name=dyn_hair_bk_l_01, bone2Name=head, swing1LimitDegrees=45, swing2LimitDegrees=45, twistLimitDegrees=5}
 {bone1Name=dyn_hair_bk_l_02, bone2Name=dyn_hair_bk_l_01, swing1LimitDegrees=45, swing2LimitDegrees=45, twistLimitDegrees=5}
 {bone1Name=dyn_hair_bk_l_03, bone2Name=dyn_hair_bk_l_02, swing1LimitDegrees=45, swing2LimitDegrees=45, twistLimitDegrees=5}
 {bone1Name=dyn_hair_bk_01, bone2Name=head, swing1LimitDegrees=45, swing2LimitDegrees=45, twistLimitDegrees=5}
 {bone1Name=dyn_hair_bk_02, bone2Name=dyn_hair_bk_01, swing1LimitDegrees=45, swing2LimitDegrees=45, twistLimitDegrees=5}
 {bone1Name=dyn_hair_bk_03, bone2Name=dyn_hair_bk_02, swing1LimitDegrees=45, swing2LimitDegrees=45, twistLimitDegrees=5}
 {bone1Name=dyn_hair_r_01, bone2Name=head, swing1LimitDegrees=45, swing2LimitDegrees=45, twistLimitDegrees=5}
 {bone1Name=dyn_hair_r_02, bone2Name=dyn_hair_r_01, swing1LimitDegrees=45, swing2LimitDegrees=45, twistLimitDegrees=5}
 {bone1Name=dyn_hair_r_03, bone2Name=dyn_hair_r_02, swing1LimitDegrees=45, swing2LimitDegrees=45, twistLimitDegrees=5}
 {
  bone1Name=dyn_hair_bk_02
  bone2Name=dyn_hair_bk_l_02
  swing1LimitDegrees=45
  swing1Motion=Free
  swing2LimitDegrees=45
  swing2Motion=Free
  twistLimitDegrees=45
  twistMotion=Free
 }
 {
  bone1Name=dyn_hair_bk_03
  bone2Name=dyn_hair_bk_l_03
  swing1LimitDegrees=45
  swing1Motion=Free
  swing2LimitDegrees=45
  swing2Motion=Free
  twistLimitDegrees=45
  twistMotion=Free
 }
 {
  bone1Name=dyn_hair_bk_l_02
  bone2Name=dyn_hair_l_02
  swing1LimitDegrees=45
  swing1Motion=Free
  swing2LimitDegrees=45
  swing2Motion=Free
  twistLimitDegrees=45
  twistMotion=Free
 }
 {
  bone1Name=dyn_hair_bk_r_02
  bone2Name=dyn_hair_r_02
  swing1LimitDegrees=45
  swing1Motion=Free
  swing2LimitDegrees=45
  swing2Motion=Free
  twistLimitDegrees=45
  twistMotion=Free
 }
 {
  bone1Name=dyn_hair_bk_02
  bone2Name=dyn_hair_bk_r_02
  swing1LimitDegrees=45
  swing1Motion=Free
  swing2LimitDegrees=45
  swing2Motion=Free
  twistLimitDegrees=45
  twistMotion=Free
 }
 {
  bone1Name=dyn_hair_bk_03
  bone2Name=dyn_hair_bk_r_03
  swing1LimitDegrees=45
  swing1Motion=Free
  swing2LimitDegrees=45
  swing2Motion=Free
  twistLimitDegrees=45
  twistMotion=Free
 }
]
