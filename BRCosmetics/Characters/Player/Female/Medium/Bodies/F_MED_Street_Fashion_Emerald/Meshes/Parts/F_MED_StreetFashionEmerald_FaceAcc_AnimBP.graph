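# Animation blueprint for the face accessory mesh of the female Street Fashion Emerald outfit, handling secondary motion and hat adjustment.
# It copies the pose from the attached parent, then runs RigidBody physics in spine_05 space with scaled, clamped component motion and overridden gravity.
# Each update maps the owner's head yaw delta from reference pose, relative to spine_05, into an alpha that blends an additive roll onto the hat bone after physics.
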AnimBlueprint /BRCosmetics/Characters/Player/Female/Medium/Bodies/F_MED_Street_Fashion_Emerald/Meshes/Parts/F_MED_StreetFashionEmerald_FaceAcc_AnimBP
target_skeleton=&/F_MED_Street_Fashion_Emerald_FaceAcc_Skeleton
graphs [
 AnimGraph {nodes [
   n1 = RigidBody(ComponentPose=Local To Component(LocalPose=Use cached pose 'inputCache'())) {
    bClampLinearTranslationLimitToRefPose=true
    bOverrideWorldGravity=true
    overridePhysicsAsset=&/F_MED_Street_Fashion_Emerald_FaceAcc_NPH
    simulationSpace=BaseBoneSpace
    alphaBoolBlend {blendOption=Linear}
    baseBoneRef {boneName=spine_05}
    componentAppliedLinearAccClamp=<5000 3000 3000>
    componentLinearAccScale=<0.46 0.3 0.3>
    componentLinearVelScale=<0.46 0.3 0.3>
    overrideWorldGravity {z=-980}
   }
   n2 = Get head_rot_back() -> (head_rot_back:double)
   Output Pose(Result=Component To Local(ComponentPose="Transform (Modify) Bone - Bone: hat"(ComponentPose=@n1, Alpha=@n2, Translation="0,0,0", Rotation="0.000,0.000,20.000", Scale="1,1,1") {rotationMode=BMM_Additive, alphaBoolBlend {blendOption=Linear}, boneToModify {boneName=hat}, rotation {roll=20}})) {layerGroup=None}
   Save cached pose 'inputCache'(Pose=Copy Pose From Mesh() {bUseAttachedParent=true}) {cacheName=inputCache}
   Use cached pose 'inputCache'()
  ]}
 EventGraph {nodes [
   Event BlueprintUpdateAnimation()
   n1 = BreakTransform(InTransform=GetDeltaTransformFromRefPose(self=GetOwnerSkeletalMeshComponent(), BoneName=head, BaseName=spine_05))
   > Set head_rot_back(head_rot_back:double=MapRangeClamped(Value=BreakRotator(InRot=@n1.Rotation).Yaw, InRangeA=0, InRangeB=-48)) -> (Output_Get:double)
   TryGetPawnOwner()
  ]}
]
variables [head_rot_back {type=number}]
